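# Conversation database for the PillowMill NPC (Gwendolyn Poole) in Fortnite, providing her greeting and branching to shared conversations.
# The entry point leads to one speech node with a default line and contextual variants gated by first-ever conversation, match conversation count, or the player's outfit.
# First-meeting and outfit lines use named priority curve rows.
# The speech then links by tag to NPC library quests, basic services, a Season 31 conversation and PillowMill dialogue.
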ConversationDatabase /NPCLibrary/NPCs/PillowMill/NPCLibrary_Conversation_PillowMill
entry_points [{tag=Conversation.Entry.PillowMill, destinations [#1]}]
nodes [
 ConversationEntryPointNode_7 {
  id=#1
  class=ConversationEntryPointNode
  outputs [#2]
  properties {entryTag {tagName=Conversation.Entry.PillowMill}}
 }
 Speech_C_4 {
  id=#2
  class=Speech_C
  outputs [#3, #4, #5, #6]
  properties {
   nodeName=""
   generalConfig {
    defaultMessage=Don't tell me Teddy followed me here too...
    contextualMessages [
     {
      message="Gwendolyn Poole, at your service. Maybe. If you're cool."
      contextRequirements [{requirement {class=FortControllerRequirement_IsFirstEverConversationWithNPC}}]
      priority {curve {rowName=Tandem.Speech.Priority.FirstEverInteraction}}
     }
     {
      message=Don't tell me Teddy followed me here too...
      contextRequirements [{requirement {class=FortControllerRequirement_CheckNPCMatchConversationCount, properties {nPCUniqueIdTag {tagName=AISpawnerData.Type.Tandem.PillowMill}}}}]
      priority {curve {rowName=None}}
     }
     {
      message=THAT'S your loadout? ...Really?
      contextRequirements [{requirement {class=FortControllerRequirement_CheckNPCMatchConversationCount, properties {nPCUniqueIdTag {tagName=AISpawnerData.Type.Tandem.PillowMill}}}}]
      priority {curve {rowName=None}}
     }
     {
      message="If I *had* to be an NPC, at least I'm in Fortnite."
      contextRequirements [{requirement {class=FortControllerRequirement_CheckNPCMatchConversationCount, properties {nPCUniqueIdTag {tagName=AISpawnerData.Type.Tandem.PillowMill}}}}]
      priority {curve {rowName=None}}
     }
     {
      message="Once again, the naming and costume similarities were a clerical error. "
      contextRequirements [{requirement {
         class=FortControllerRequirement_HasCID
         properties {allowedCIds [
           /BRCosmetics/Athena/Items/Cosmetics/Characters/CID_745_Athena_Commando_M_RavenQuill
           /BRCosmetics/Athena/Items/Cosmetics/Characters/CID_746_Athena_Commando_F_FuzzyBear
           /BRCosmetics/Athena/Items/Cosmetics/Characters/CID_705_Athena_Commando_M_Donut
          ]}
        }}]
      priority {curve {rowName=Tandem.Speech.Priority.OutfitReaction}}
     }
     {
      message=It's the GwenHive all over again.
      contextRequirements [{requirement {
         class=FortControllerRequirement_HasCID
         properties {allowedCIds [/BRCosmetics/Athena/Items/Cosmetics/Characters/Character_PillowMill_Mastery]}
        }}]
      priority {curve {rowName=Tandem.Speech.Priority.OutfitReaction}}
     }
    ]
   }
  }
 }
 ConversationLinkNode_3 {
  id=#6
  class=ConversationLinkNode
  properties {remoteEntryTag {tagName=Conversation.Entry.PillowMill.Dialogue}}
 }
 ConversationLinkNode_2 {
  id=#5
  class=ConversationLinkNode
  properties {remoteEntryTag {tagName=Conversation.Entry.Season31.Found.PB.P02}}
 }
 ConversationLinkNode_0 {id=#4, class=ConversationLinkNode, properties {remoteEntryTag {tagName=Conversation.Entry.BasicServices}}}
 ConversationLinkNode_1 {id=#3, class=ConversationLinkNode, properties {remoteEntryTag {tagName=Conversation.Entry.NPCLibrary.Quests}}}
]
